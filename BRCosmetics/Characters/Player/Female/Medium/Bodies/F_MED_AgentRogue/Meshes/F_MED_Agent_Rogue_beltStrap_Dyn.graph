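# Physics asset for the dynamic belt strap on the medium female Agent Rogue player body.
# A box at the strap root leads a chain of four capsules, each constrained to the previous with moderate swing and very tight twist limits.
# Capsules on the pelvis and right upper thigh give body shapes around the strap.
# The strap root and thigh both attach to the pelvis with free angular motion.
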PhysicsAsset /BRCosmetics/Characters/Player/Female/Medium/Bodies/F_MED_AgentRogue/Meshes/F_MED_Agent_Rogue_beltStrap_Dyn
bodies {
 dyn_belt_strap_01 [{shapeType=Box, extentX=3, extentY=3, extentZ=1.2}]
 dyn_belt_strap_02 [{shapeType=Capsule, length=5.8717, radius=2.2, center=<5.7198 0 0>, rotation=<rot 90 180 -174.022>}]
 dyn_belt_strap_03 [{shapeType=Capsule, length=7.75, radius=2.2, center=<4.4887 0 0>, rotation=<rot 90 0 0>}]
 dyn_belt_strap_04 [{shapeType=Capsule, length=8.5, radius=2.2, center=<5.1146 0 0>, rotation=<rot 90 0 0>}]
 dyn_belt_strap_05 [{shapeType=Capsule, length=7.5, radius=2.45, center=<5.1146 0 0>, rotation=<rot 90 0 0>}]
 pelvis [{shapeType=Capsule, length=12.25, radius=8, center=<0.85 -1.9 0>}]
 thigh_twist_01_r [{shapeType=Capsule, length=24.5, radius=7.75, center=<-13.3142 1.2437 0>, rotation=<rot 90 180 -178>}]
}
constraints [
 {
  bone1Name=dyn_belt_strap_01
  bone2Name=pelvis
  swing1LimitDegrees=45
  swing1Motion=Free
  swing2LimitDegrees=45
  swing2Motion=Free
  twistLimitDegrees=45
  twistMotion=Free
 }
 {bone1Name=dyn_belt_strap_02, bone2Name=dyn_belt_strap_01, swing1LimitDegrees=60, swing2LimitDegrees=60, twistLimitDegrees=5}
 {bone1Name=dyn_belt_strap_03, bone2Name=dyn_belt_strap_02, swing1LimitDegrees=45, swing2LimitDegrees=45, twistLimitDegrees=5}
 {bone1Name=dyn_belt_strap_04, bone2Name=dyn_belt_strap_03, swing1LimitDegrees=45, swing2LimitDegrees=45, twistLimitDegrees=5}
 {bone1Name=dyn_belt_strap_05, bone2Name=dyn_belt_strap_04, swing1LimitDegrees=45, swing2LimitDegrees=45, twistLimitDegrees=5}
 {
  bone1Name=thigh_twist_01_r
  bone2Name=pelvis
  swing1LimitDegrees=45
  swing1Motion=Free
  swing2LimitDegrees=45
  swing2Motion=Free
  twistLimitDegrees=45
  twistMotion=Free
 }
]
