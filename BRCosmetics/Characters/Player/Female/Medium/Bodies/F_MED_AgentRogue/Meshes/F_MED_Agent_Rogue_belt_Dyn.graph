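# Physics asset for the Agent Rogue female medium body that sets up secondary motion for three left-side belt bones.
# Each belt box is constrained to its own small MSTR capsule with tight swing1 and twist limits and a wider swing2, giving a mostly hinge-like flap.
# The MSTR capsules link to the pelvis with free rotation.
# Capsules on the pelvis and left upper thigh give body shapes for the belt pieces to collide against.
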PhysicsAsset /BRCosmetics/Characters/Player/Female/Medium/Bodies/F_MED_AgentRogue/Meshes/F_MED_Agent_Rogue_belt_Dyn
bodies {
 dyn_belt_l_01 [{shapeType=Box, extentX=7.5, extentY=10, extentZ=3, center=<3 0 1.5>}]
 dyn_belt_l_01_MSTR [{shapeType=Capsule, length=1, radius=1, rotation=<rot 90 0 0>}]
 dyn_belt_l_02 [{shapeType=Box, extentX=7.5, extentY=10, extentZ=3, center=<3 0 1.5>}]
 dyn_belt_l_02_MSTR [{shapeType=Capsule, length=1, radius=1, rotation=<rot 90 0 0>}]
 dyn_belt_l_03 [{shapeType=Box, extentX=7.5, extentY=10, extentZ=3, center=<3 0 1.5>}]
 dyn_belt_l_03_MSTR [{shapeType=Capsule, length=1, radius=1, rotation=<rot 90 0 0>}]
 pelvis [{shapeType=Capsule, length=9.5, radius=8.9, center=<0 -2 0>}]
 thigh_twist_01_l [{shapeType=Capsule, length=16, radius=8, center=<17 -1 0>, rotation=<rot 85 175 180>}]
}
constraints [
 {bone1Name=dyn_belt_l_01, bone2Name=dyn_belt_l_01_MSTR, swing1LimitDegrees=5, swing2LimitDegrees=30, twistLimitDegrees=5}
 {bone1Name=dyn_belt_l_02, bone2Name=dyn_belt_l_02_MSTR, swing1LimitDegrees=5, swing2LimitDegrees=30, twistLimitDegrees=5}
 {bone1Name=dyn_belt_l_03, bone2Name=dyn_belt_l_03_MSTR, swing1LimitDegrees=5, swing2LimitDegrees=30, twistLimitDegrees=5}
 {
  bone1Name=dyn_belt_l_03_MSTR
  bone2Name=pelvis
  swing1LimitDegrees=45
  swing1Motion=Free
  swing2LimitDegrees=45
  swing2Motion=Free
  twistLimitDegrees=45
  twistMotion=Free
 }
 {
  bone1Name=dyn_belt_l_02_MSTR
  bone2Name=pelvis
  swing1LimitDegrees=45
  swing1Motion=Free
  swing2LimitDegrees=45
  swing2Motion=Free
  twistLimitDegrees=45
  twistMotion=Free
 }
 {
  bone1Name=dyn_belt_l_01_MSTR
  bone2Name=pelvis
  swing1LimitDegrees=45
  swing1Motion=Free
  swing2LimitDegrees=45
  swing2Motion=Free
  twistLimitDegrees=45
  twistMotion=Free
 }
 {
  bone1Name=thigh_twist_01_l
  bone2Name=pelvis
  swing1LimitDegrees=45
  swing1Motion=Free
  swing2LimitDegrees=45
  swing2Motion=Free
  twistLimitDegrees=45
  twistMotion=Free
 }
]
