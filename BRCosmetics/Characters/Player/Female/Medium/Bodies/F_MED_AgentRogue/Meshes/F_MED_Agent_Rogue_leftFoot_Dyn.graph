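# Physics asset for the left foot of the female medium Agent Rogue character body, set up for dynamic shoelace motion.
# It defines box bodies on two left shoelace bones and a capsule body on the left foot bone.
# Each lace is constrained directly to the foot rather than chained to the other, with tight twist, narrow first swing and a wider second swing limit.
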PhysicsAsset /BRCosmetics/Characters/Player/Female/Medium/Bodies/F_MED_AgentRogue/Meshes/F_MED_Agent_Rogue_leftFoot_Dyn
bodies {
 dyn_shoelaces_l_01 [{shapeType=Box, extentX=4, extentY=3, extentZ=1.5, center=<2 0 0>}]
 dyn_shoelaces_l_02 [{shapeType=Box, extentX=4, extentY=3, extentZ=1.5, center=<2 0 0>}]
 foot_l [{shapeType=Capsule, length=8.2, radius=3.8, center=<-3.7 7 0.6>, rotation=<rot 90 0 65>}]
}
constraints [
 {bone1Name=dyn_shoelaces_l_01, bone2Name=foot_l, swing1LimitDegrees=10, swing2LimitDegrees=60, twistLimitDegrees=5}
 {bone1Name=dyn_shoelaces_l_02, bone2Name=foot_l, swing1LimitDegrees=10, swing2LimitDegrees=60, twistLimitDegrees=5}
]
