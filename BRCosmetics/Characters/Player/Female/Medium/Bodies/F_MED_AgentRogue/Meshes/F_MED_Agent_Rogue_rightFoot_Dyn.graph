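# Physics asset for the Agent Rogue female medium body's right foot, setting up dynamic shoelace simulation.
# Two small box bodies on the shoelace bones hang off a capsule on foot_r.
# Each lace is constrained directly to the foot rather than chained to the other, with tight twist and first swing limits and a much wider second swing.
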PhysicsAsset /BRCosmetics/Characters/Player/Female/Medium/Bodies/F_MED_AgentRogue/Meshes/F_MED_Agent_Rogue_rightFoot_Dyn
bodies {
 dyn_shoelaces_r_01 [{shapeType=Box, extentX=4, extentY=3, extentZ=1.5, center=<2 0 0>}]
 dyn_shoelaces_r_02 [{shapeType=Box, extentX=4, extentY=3, extentZ=1.5, center=<2 0 0>}]
 foot_r [{shapeType=Capsule, length=8.2, radius=3.8, center=<3.7 -7 -0.4108>, rotation=<rot 90 -90 -24.9995>}]
}
constraints [
 {bone1Name=dyn_shoelaces_r_02, bone2Name=foot_r, swing1LimitDegrees=10, swing2LimitDegrees=60, twistLimitDegrees=5}
 {bone1Name=dyn_shoelaces_r_01, bone2Name=foot_r, swing1LimitDegrees=10, swing2LimitDegrees=60, twistLimitDegrees=5}
]
