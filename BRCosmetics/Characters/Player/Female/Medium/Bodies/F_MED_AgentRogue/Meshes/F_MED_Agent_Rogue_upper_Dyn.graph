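# Physics asset for the upper-body mesh of the female medium Agent Rogue character, covering dynamic necklace and strap bones.
# Capsule bodies cover a two-bone necklace chain, a three-bone strap chain, and two spine bones that form the torso volume.
# The necklace chains from spine_05 with narrow swing and twist limits.
# The straps root to spine_03 through an angularly free joint, then use limited joints; spine_05 also joins spine_03 freely.
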PhysicsAsset /BRCosmetics/Characters/Player/Female/Medium/Bodies/F_MED_AgentRogue/Meshes/F_MED_Agent_Rogue_upper_Dyn
bodies {
 dyn_necklace_01 [{
   shapeType=Capsule
   length=3.3032
   radius=3.112
   shapeName=dyn_necklace_01_capsule
   center=<2.8831 0 2.5>
   rotation=<rot -87.9935 179.9999 -179.9999>
  }]
 dyn_necklace_02 [{
   shapeType=Capsule
   length=3.3032
   radius=3.112
   shapeName=dyn_necklace_02_capsule
   center=<2.8831 0 2.5>
   rotation=<rot -88.0134 179.9994 -179.9994>
  }]
 dyn_straps_01 [{shapeType=Capsule, length=1, radius=1, rotation=<rot 90 0 0>}]
 dyn_straps_02 [{shapeType=Capsule, length=2, radius=1.75, shapeName=dyn_straps_02_capsule, center=<2 0 0>, rotation=<rot 90 0 0>}]
 dyn_straps_03 [{shapeType=Capsule, length=2, radius=1.75, shapeName=dyn_straps_03_capsule, center=<2 0 0>, rotation=<rot 90 0 0>}]
 spine_03 [{shapeType=Capsule, length=10.3, radius=8.05, center=<0.6633 1.4584 0>}]
 spine_05 [{shapeType=Capsule, length=11.5, radius=7.2, center=<3 -1.5 0>}, {shapeType=Capsule, length=11, radius=6.5, center=<0.5 2 0>}]
}
constraints [
 {bone1Name=dyn_necklace_01, bone2Name=spine_05, swing1LimitDegrees=15, swing2LimitDegrees=30, twistLimitDegrees=2}
 {bone1Name=dyn_necklace_02, bone2Name=dyn_necklace_01, swing1LimitDegrees=15, swing2LimitDegrees=30, twistLimitDegrees=2}
 {
  bone1Name=spine_05
  bone2Name=spine_03
  swing1LimitDegrees=45
  swing1Motion=Free
  swing2LimitDegrees=45
  swing2Motion=Free
  twistLimitDegrees=45
  twistMotion=Free
 }
 {
  bone1Name=dyn_straps_01
  bone2Name=spine_03
  swing1LimitDegrees=45
  swing1Motion=Free
  swing2LimitDegrees=45
  swing2Motion=Free
  twistLimitDegrees=45
  twistMotion=Free
 }
 {bone1Name=dyn_straps_02, bone2Name=dyn_straps_01, swing1LimitDegrees=35, swing2LimitDegrees=35, twistLimitDegrees=5}
 {bone1Name=dyn_straps_03, bone2Name=dyn_straps_02, swing1LimitDegrees=35, swing2LimitDegrees=35, twistLimitDegrees=5}
]
