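# Post-process style AnimBP for the female Agent Rogue body that adds corrective and secondary motion to a pose copied from the attached parent mesh.
# Each update maps thigh and right upper-arm rotations to alphas driving additive glute, belt and shoulder-armor corrections.
# Constraints position armor and straps, while RigidBody and AnimDynamics nodes simulate straps, belt pieces and shoelaces.
# Separate belt and foot RigidBody nodes are unconnected to the output.
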AnimBlueprint /BRCosmetics/Characters/Player/Female/Medium/Bodies/F_MED_AgentRogue/Meshes/F_MED_Agent_Rogue_AnimBP
target_skeleton=&/F_MED_Agent_Rogue_Skeleton
graphs [
 AnimGraph {nodes [
   RigidBody() {
    bClampLinearTranslationLimitToRefPose=true
    bOverrideWorldGravity=true
    overridePhysicsAsset=&/F_MED_Agent_Rogue_belt_Dyn
    simulationSpace=BaseBoneSpace
    alphaBoolBlend {blendOption=Linear}
    baseBoneRef {boneName=pelvis}
    componentAppliedLinearAccClamp=<2000 2000 2000>
    componentLinearAccScale=<0.25 0.3 0.3>
    componentLinearVelScale=<0.2 0.2 0.2>
    overrideWorldGravity {z=-980}
   }
   n1 = RigidBody() {
    bClampLinearTranslationLimitToRefPose=true
    bOverrideWorldGravity=true
    overridePhysicsAsset=&/F_MED_Agent_Rogue_leftFoot_Dyn
    simulationSpace=BaseBoneSpace
    alphaBoolBlend {blendOption=Linear}
    baseBoneRef {boneName=foot_l}
    componentAppliedLinearAccClamp=<3000 3000 3000>
    componentLinearAccScale=<0.65 0.65 0.3>
    componentLinearVelScale=<0.3 0.3 0.3>
    overrideWorldGravity {z=-980}
   }
   RigidBody(ComponentPose=@n1) {
    bClampLinearTranslationLimitToRefPose=true
    bOverrideWorldGravity=true
    overridePhysicsAsset=&/F_MED_Agent_Rogue_rightFoot_Dyn
    simulationSpace=BaseBoneSpace
    alphaBoolBlend {blendOption=Linear}
    baseBoneRef {boneName=foot_r}
    componentAppliedLinearAccClamp=<3000 3000 3000>
    componentLinearAccScale=<0.65 0.65 0.3>
    componentLinearVelScale=<0.3 0.3 0.3>
    overrideWorldGravity {z=-980}
   }
   n2 = "Transform (Modify) Bone - Bone: dyn_glute_l"(
    ComponentPose=Local To Component(LocalPose=Copy Pose From Mesh() {bUseAttachedParent=true})
    Alpha=Get thigh_rot_up_l() -> (thigh_rot_up_l:double)
    Translation="-9,0,0"
    Rotation="0,0,0"
    Scale="1,1,1"
   ) {
    translationMode=BMM_Additive
    translationSpace=BCS_BoneSpace
    alphaBoolBlend {blendOption=Linear}
    boneToModify {boneName=dyn_glute_l}
    translation {x=-9}
   }
   n3 = "Transform (Modify) Bone - Bone: dyn_glute_r"(ComponentPose=@n2, Alpha=Get thigh_rot_up_r() -> (thigh_rot_up_r:double), Translation="-9,0,0", Rotation="0,0,0", Scale="1,1,1") {
    translationMode=BMM_Additive
    translationSpace=BCS_BoneSpace
    alphaBoolBlend {blendOption=Linear}
    boneToModify {boneName=dyn_glute_r}
    translation {x=-9}
   }
   n4 = "Constraint - dyn_shoulder_armor_r_MSTR (PPP)"!(ComponentPose=@n3, Alpha=1, ConstraintWeights_0=0.016, ConstraintWeights_1:float=0.967, ConstraintWeights_2:float=0.017) -> (Pose) {
    alphaBoolBlend {blendOption=Linear}
    boneToModify {boneName=dyn_shoulder_armor_r_MSTR}
    constraintSetup [{targetBone {boneName=upperarm_r}}, {targetBone {boneName=upperarm_twist_01_r}}, {targetBone {boneName=upperarm_twist_02_r}}]
    constraintWeights [1,1,1]
   }
   n5 = "Transform (Modify) Bone - Bone: dyn_shoulder_armor_r"(ComponentPose=@n4, Alpha=Get arm_rot_up_r() -> (arm_rot_up_r:double), Translation="-3.5,0,0.000", Rotation="8,0,0", Scale="1,1,1") {
    rotationMode=BMM_Additive
    rotationSpace=BCS_BoneSpace
    translationMode=BMM_Additive
    translationSpace=BCS_BoneSpace
    alphaBoolBlend {blendOption=Linear}
    boneToModify {boneName=dyn_shoulder_armor_r}
    rotation {pitch=8}
    translation {x=-3.5}
   }
   n6 = Get thigh_rot_up_l()
   n7 = "Transform (Modify) Bone - Bone: dyn_belt_l_01_MSTR"(ComponentPose=@n5, Alpha=@n6, Translation="-2.95,0.622,1.385", Rotation="100,-20,3", Scale="1,1,1") {
    rotationMode=BMM_Additive
    rotationSpace=BCS_BoneSpace
    translationMode=BMM_Additive
    translationSpace=BCS_BoneSpace
    alphaBoolBlend {blendOption=Linear}
    boneToModify {boneName=dyn_belt_l_01_MSTR}
    rotation=<rot 100 -20 3>
    translation=<-2.95 0.622 1.385>
   }
   n8 = "Transform (Modify) Bone - Bone: dyn_belt_l_02_MSTR"(ComponentPose=@n7, Alpha=@n6, Translation="-2.9,0.25,1.68", Rotation="66.000,55.000,-3", Scale="1,1,1") {
    rotationMode=BMM_Additive
    rotationSpace=BCS_BoneSpace
    translationMode=BMM_Additive
    translationSpace=BCS_BoneSpace
    alphaBoolBlend {blendOption=Linear}
    boneToModify {boneName=dyn_belt_l_02_MSTR}
    rotation=<rot 66 55 -3>
    translation=<-2.9 0.25 1.68>
   }
   n9 = Anim Dynamics - Bone: dyn_belt_l_01!(ComponentPose=@n8, Alpha=1, GravityScale:float=1, GravityOverride:Vector="0,0,-980", ExternalForce="0,0,0") -> (Pose) {
    angularDampingOverride=0.9999
    angularSpringConstant=1000
    bAngularSpring=true
    bGravityOverrideInSimSpace=true
    bOverrideAngularDamping=true
    simulationSpace=RootRelative
    alphaBoolBlend {blendOption=Linear}
    boundBone {boneName=dyn_belt_l_01}
    componentAppliedLinearAccClamp=<5000 5000 5000>
    componentLinearAccScale=<0.6 0.6 0.6>
    componentLinearVelScale=<0.6 0.6 0.6>
    gravityOverride {z=-980}
    physicsBodyDefinitions [{
      sphereCollisionRadius=0
      boundBone {boneName=dyn_belt_l_01}
      boxExtents=<12 20 10>
      constraintSetup {angularLimitsMax=<0 2 5>, angularLimitsMin=<0 -35 -5>, angularTarget=<0 0 0>}
      localJointOffset=<1.5 0 0>
     }]
    relativeSpaceBone {boneName=pelvis}
   }
   n10 = Anim Dynamics - Bone: dyn_belt_l_02!(ComponentPose=@n9, Alpha=1, GravityScale:float=1, GravityOverride:Vector="0,0,-980", ExternalForce="0,0,0") -> (Pose) {
    angularDampingOverride=0.9999
    angularSpringConstant=1000
    bAngularSpring=true
    bGravityOverrideInSimSpace=true
    bOverrideAngularDamping=true
    simulationSpace=RootRelative
    alphaBoolBlend {blendOption=Linear}
    boundBone {boneName=dyn_belt_l_02}
    componentAppliedLinearAccClamp=<5000 5000 5000>
    componentLinearAccScale=<0.6 0.6 0.6>
    componentLinearVelScale=<0.6 0.6 0.6>
    gravityOverride {z=-980}
    physicsBodyDefinitions [{
      sphereCollisionRadius=0
      boundBone {boneName=dyn_belt_l_02}
      boxExtents=<12 20 10>
      constraintSetup {angularLimitsMax=<0 2 5>, angularLimitsMin=<0 -35 -5>, angularTarget=<0 0 0>}
      localJointOffset=<1.5 0 -0.3>
     }]
    relativeSpaceBone {boneName=pelvis}
   }
   n11 = Anim Dynamics - Bone: dyn_belt_l_03!(ComponentPose=@n10, Alpha=1, GravityScale:float=1, GravityOverride:Vector="0,0,-980", ExternalForce="0,0,0") -> (Pose) {
    angularDampingOverride=0.9999
    angularSpringConstant=1000
    bAngularSpring=true
    bGravityOverrideInSimSpace=true
    bOverrideAngularDamping=true
    simulationSpace=RootRelative
    alphaBoolBlend {blendOption=Linear}
    boundBone {boneName=dyn_belt_l_03}
    componentAppliedLinearAccClamp=<5000 5000 5000>
    componentLinearAccScale=<0.6 0.6 0.6>
    componentLinearVelScale=<0.6 0.6 0.6>
    gravityOverride {z=-980}
    physicsBodyDefinitions [{
      sphereCollisionRadius=0
      boundBone {boneName=dyn_belt_l_03}
      boxExtents=<12 20 10>
      constraintSetup {angularLimitsMax=<0 2 5>, angularLimitsMin=<0 -35 -5>, angularTarget=<0 0 0>}
      localJointOffset=<1.5 0 -0.8>
     }]
    relativeSpaceBone {boneName=pelvis}
   }
   n12 = "Transform (Modify) Bone - Bone: dyn_belt_strap_01"(ComponentPose=@n11, Alpha=Get thigh_rot_up_r(), Translation="0,1,0", Rotation="-65.000,-0.000,70.000", Scale="1,1,1") {
    rotationMode=BMM_Additive
    rotationSpace=BCS_BoneSpace
    translationMode=BMM_Additive
    translationSpace=BCS_BoneSpace
    alphaBoolBlend {blendOption=Linear}
    boneToModify {boneName=dyn_belt_strap_01}
    rotation=<rot -65 -0.0001 70.0001>
    translation {y=1}
   }
   n13 = RigidBody(ComponentPose=@n12) {
    bClampLinearTranslationLimitToRefPose=true
    bOverrideWorldGravity=true
    overridePhysicsAsset=&/F_MED_Agent_Rogue_beltStrap_Dyn
    simulationSpace=BaseBoneSpace
    alphaBoolBlend {blendOption=Linear}
    baseBoneRef {boneName=pelvis}
    componentAppliedLinearAccClamp=<1000 1000 1000>
    componentLinearAccScale=<0.3 0.1 0.1>
    componentLinearVelScale=<0.5 0.5 0.5>
    overrideWorldGravity {z=-980}
   }
   n14 = "Constraint - dyn_straps_MSTR (PPP)"!(ComponentPose=@n13, Alpha=1, ConstraintWeights_0=0.013, ConstraintWeights_1:float=0.945, ConstraintWeights_2:float=0.042) -> (Pose) {
    alphaBoolBlend {blendOption=Linear}
    boneToModify {boneName=dyn_straps_MSTR}
    constraintSetup [{targetBone {boneName=spine_03}}, {targetBone {boneName=spine_04}}, {targetBone {boneName=spine_05}}]
    constraintWeights [1,1,1]
   }
   n15 = RigidBody(ComponentPose=@n14) {
    bClampLinearTranslationLimitToRefPose=true
    bOverrideWorldGravity=true
    overridePhysicsAsset=&/F_MED_Agent_Rogue_upper_Dyn
    simulationSpace=BaseBoneSpace
    alphaBoolBlend {blendOption=Linear}
    baseBoneRef {boneName=spine_03}
    componentAppliedLinearAccClamp=<2000 2000 2000>
    componentLinearAccScale=<0.35 0.3 0.3>
    componentLinearVelScale=<0.4 0.4 0.4>
    overrideWorldGravity {z=-980}
   }
   n16 = Anim Dynamics - Bone: dyn_shoelaces_l_01!(ComponentPose=@n15, Alpha=1, GravityScale:float=1, GravityOverride:Vector="0,0,-980", ExternalForce="0,0,0") -> (Pose) {
    angularDampingOverride=0.9999
    angularSpringConstant=1000
    bAngularSpring=true
    bGravityOverrideInSimSpace=true
    bOverrideAngularDamping=true
    simulationSpace=RootRelative
    alphaBoolBlend {blendOption=Linear}
    boundBone {boneName=dyn_shoelaces_l_01}
    componentAppliedLinearAccClamp=<5000 5000 5000>
    componentLinearAccScale=<0.6 0.6 0.6>
    componentLinearVelScale=<0.6 0.6 0.6>
    gravityOverride {z=-980}
    physicsBodyDefinitions [{
      sphereCollisionRadius=0
      boundBone {boneName=dyn_shoelaces_l_01}
      boxExtents=<12 10 10>
      constraintSetup {angularLimitsMax=<0 5 15>, angularLimitsMin=<0 -35 -15>, angularTarget=<0 0 0>}
      localJointOffset=<1.5 0 -1.5>
     }]
    relativeSpaceBone {boneName=foot_l}
   }
   n17 = Anim Dynamics - Bone: dyn_shoelaces_l_02!(ComponentPose=@n16, Alpha=1, GravityScale:float=1, GravityOverride:Vector="0,0,-980", ExternalForce="0,0,0") -> (Pose) {
    angularDampingOverride=0.999
    angularSpringConstant=1000
    bAngularSpring=true
    bGravityOverrideInSimSpace=true
    bOverrideAngularDamping=true
    simulationSpace=RootRelative
    alphaBoolBlend {blendOption=Linear}
    boundBone {boneName=dyn_shoelaces_l_02}
    componentAppliedLinearAccClamp=<5000 5000 5000>
    componentLinearAccScale=<0.6 0.6 0.6>
    componentLinearVelScale=<0.6 0.6 0.6>
    gravityOverride {z=-980}
    physicsBodyDefinitions [{
      sphereCollisionRadius=0
      boundBone {boneName=dyn_shoelaces_l_02}
      boxExtents=<12 10 10>
      constraintSetup {angularLimitsMax=<0 5 15>, angularLimitsMin=<0 -35 -15>, angularTarget=<0 0 0>}
      localJointOffset=<1.5 0 -2>
     }]
    relativeSpaceBone {boneName=foot_l}
   }
   n18 = Anim Dynamics - Bone: dyn_shoelaces_r_01!(ComponentPose=@n17, Alpha=1, GravityScale:float=1, GravityOverride:Vector="0,0,-980", ExternalForce="0,0,0") -> (Pose) {
    angularDampingOverride=0.9999
    angularSpringConstant=1000
    bAngularSpring=true
    bGravityOverrideInSimSpace=true
    bOverrideAngularDamping=true
    simulationSpace=RootRelative
    alphaBoolBlend {blendOption=Linear}
    boundBone {boneName=dyn_shoelaces_r_01}
    componentAppliedLinearAccClamp=<5000 5000 5000>
    componentLinearAccScale=<0.6 0.6 0.6>
    componentLinearVelScale=<0.6 0.6 0.6>
    gravityOverride {z=-980}
    physicsBodyDefinitions [{
      sphereCollisionRadius=0
      boundBone {boneName=dyn_shoelaces_r_01}
      boxExtents=<12 10 10>
      constraintSetup {angularLimitsMax=<0 5 15>, angularLimitsMin=<0 -35 -15>, angularTarget=<0 0 0>}
      localJointOffset=<1.5 0 -1.5>
     }]
    relativeSpaceBone {boneName=foot_r}
   }
   n19 = Anim Dynamics - Bone: dyn_shoelaces_r_02!(ComponentPose=@n18, Alpha=1, GravityScale:float=1, GravityOverride:Vector="0,0,-980", ExternalForce="0,0,0") -> (Pose) {
    angularDampingOverride=0.999
    angularSpringConstant=1000
    bAngularSpring=true
    bGravityOverrideInSimSpace=true
    bOverrideAngularDamping=true
    simulationSpace=RootRelative
    alphaBoolBlend {blendOption=Linear}
    boundBone {boneName=dyn_shoelaces_r_02}
    componentAppliedLinearAccClamp=<5000 5000 5000>
    componentLinearAccScale=<0.6 0.6 0.6>
    componentLinearVelScale=<0.6 0.6 0.6>
    gravityOverride {z=-980}
    physicsBodyDefinitions [{
      sphereCollisionRadius=0
      boundBone {boneName=dyn_shoelaces_r_02}
      boxExtents=<12 10 10>
      constraintSetup {angularLimitsMax=<0 5 15>, angularLimitsMin=<0 -35 -15>, angularTarget=<0 0 0>}
      localJointOffset=<1.5 0 -2>
     }]
    relativeSpaceBone {boneName=foot_r}
   }
   Output Pose(Result=Component To Local(ComponentPose=@n19)) {layerGroup=None}
  ]}
 EventGraph {nodes [
   Event BlueprintUpdateAnimation()
   > n1 = Sequence!(execute) -> (then_0, then_1, then_2:Exec)
   n2 = BreakTransform(InTransform=GetDeltaTransformFromRefPose(self=GetOwnerSkeletalMeshComponent(), BoneName=thigh_r, BaseName=pelvis))
   > Set thigh_rot_up_r(thigh_rot_up_r:double=MapRangeClamped(Value=BreakRotator(InRot=@n2.Rotation).Yaw, InRangeA=0, InRangeB=-100)) -> (Output_Get:double)
   n3 = BreakTransform(InTransform=GetDeltaTransformFromRefPose(self=GetOwnerSkeletalMeshComponent(), BoneName=thigh_l, BaseName=pelvis))
   > Set thigh_rot_up_l(thigh_rot_up_l:double=MapRangeClamped(Value=BreakRotator(InRot=@n3.Rotation).Yaw, InRangeA=-30, InRangeB=-100)) -> (Output_Get:double)
   n4 = BreakTransform(InTransform=GetDeltaTransformFromRefPose(self=GetOwnerSkeletalMeshComponent(), BoneName=upperarm_r, BaseName=clavicle_r))
   Set arm_rot_up_r(execute=@n1.then_1, arm_rot_up_r:double=MapRangeClamped(Value=BreakRotator(InRot=@n4.Rotation).Pitch, InRangeA=0, InRangeB=70)) -> (Output_Get:double)
  ]}
]
variables [
 arm_rot_up_r {type=number}
 thigh_rot_up_l {type=number}
 thigh_rot_up_r {type=number}
]
