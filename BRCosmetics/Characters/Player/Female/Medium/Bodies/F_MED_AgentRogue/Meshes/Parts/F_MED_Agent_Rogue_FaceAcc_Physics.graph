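# Physics asset for the female Agent Rogue face accessory mesh, simulating dynamic hair with capsule bodies on the hair, head, neck, upper spine, clavicles and right upper arm.
# Three back chains hang from master bodies on the head; front-left attaches to the head, front-right to the neck.
# Hair segments use limited swing and tight twist, while body and master joints rotate freely.
# Angular-locked cross-constraints tie the side back chains' tips to the centre chain.
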PhysicsAsset /BRCosmetics/Characters/Player/Female/Medium/Bodies/F_MED_AgentRogue/Meshes/Parts/F_MED_Agent_Rogue_FaceAcc_Physics
bodies {
 clavicle_l [{shapeType=Capsule, length=4.488, radius=5.6474, center=<4.048 0 0>, rotation=<rot 90 0 0>}]
 clavicle_r [{shapeType=Capsule, length=4.488, radius=5.6474, center=<-4.048 0 0>, rotation=<rot 90 0 0>}]
 dyn_hair_bk_01 [{shapeType=Capsule, length=0.75, radius=3.25, center=<3 0 1>, rotation=<rot 90 0 0>}]
 dyn_hair_bk_02 [{shapeType=Capsule, length=3.25, radius=3.25, center=<3 0 1>, rotation=<rot 0 0 90>}]
 dyn_hair_bk_03 [{shapeType=Capsule, length=3.7707, radius=2.75, center=<3 0 1.2>, rotation=<rot 0 0 90>}]
 dyn_hair_bk_04 [{shapeType=Capsule, length=3.7, radius=1.75, center=<2 0 1.5>, rotation=<rot 0 0 90>}]
 dyn_hair_bk_MSTR [{shapeType=Capsule, length=0.505, radius=0.505, shapeName=dyn_hair_bk_MSTR_capsule, rotation=<rot 0 -90 0>}]
 dyn_hair_bk_l_01 [{shapeType=Capsule, length=0.75, radius=3.25, center=<3 0 1>, rotation=<rot 90 0 0>}]
 dyn_hair_bk_l_02 [{shapeType=Capsule, length=2.75, radius=3.75, center=<3.1543 0.5 1.101>, rotation=<rot 85.0001 -180 180>}]
 dyn_hair_bk_l_03 [{shapeType=Capsule, length=2, radius=3.05, center=<2 0 2>, rotation=<rot 75.0003 179.9997 179.9996>}]
 dyn_hair_bk_l_04 [{shapeType=Capsule, length=2.5, radius=2.25, center=<1 0 2>, rotation=<rot 85.0001 179.9998 179.9998>}]
 dyn_hair_bk_l_MSTR [{shapeType=Capsule, length=0.505, radius=0.505, shapeName=dyn_hair_bk_l_MSTR_capsule, rotation=<rot 0 -90 0>}]
 dyn_hair_bk_r_01 [{shapeType=Capsule, length=0.75, radius=3.25, center=<3 0 1>, rotation=<rot 90 0 0>}]
 dyn_hair_bk_r_02 [{shapeType=Capsule, length=2.75, radius=2.75, center=<2 0 1>, rotation=<rot 85.0001 179.9998 179.9999>}]
 dyn_hair_bk_r_03 [{shapeType=Capsule, length=1, radius=2.2, center=<1 0 2>, rotation=<rot 90 0 0>}]
 dyn_hair_bk_r_MSTR [{shapeType=Capsule, length=0.505, radius=0.505, shapeName=dyn_hair_bk_r_MSTR_capsule, rotation=<rot 0 -90 0>}]
 dyn_hair_fr_l_01 [{shapeType=Capsule, length=2.5, radius=2.5, shapeName=dyn_hair_fr_l_01_capsule, center=<2 0 0.5>, rotation=<rot 90 0 0>}]
 dyn_hair_fr_l_02 [{shapeType=Capsule, length=2.25, radius=2.75, shapeName=dyn_hair_fr_l_02_capsule, center=<2 0 0>, rotation=<rot 85 0.0001 0>}]
 dyn_hair_fr_l_03 [{shapeType=Capsule, length=3.75, radius=2.25, center=<3.595 3 0>, rotation=<rot 90 90 64.9997>}]
 dyn_hair_fr_r_01 [{shapeType=Capsule, length=3.2, radius=2.25, center=<3.0873 0 0>, rotation=<rot 90 -150.2589 -135.2588>}]
 dyn_hair_fr_r_02 [{shapeType=Capsule, length=2.6884, radius=2.2291, center=<2.6583 -1.5 1.5>, rotation=<rot 60.0639 171.5433 -178.2616>}]
 dyn_hair_fr_r_03 [{shapeType=Capsule, length=2.0773, radius=1.9227, center=<1.6451 0 2.5>, rotation=<rot 79.9999 179.9997 179.9998>}]
 head [{shapeType=Capsule, length=2.1322, radius=5.217, center=<0.4562 0.66 0>, rotation=<rot 90 0 0>}]
 neck_01 [{shapeType=Capsule, length=2.7301, radius=5.5, center=<1.158 0.0001 0>, rotation=<rot 90 0 0>}]
 spine_04 [{shapeType=Capsule, length=9.2471, radius=9.9074, center=<6.8107 0.479 -0.0059>, rotation=<rot -0.1295 -1.4854 -0.054>}]
 spine_05 [{shapeType=Capsule, length=13.7471, radius=7.9074, center=<4.3281 -0.4885 -0.0013>, rotation=<rot -0.1295 -1.4854 -0.054>}]
 upperarm_r [{shapeType=Capsule, length=5.488, radius=6.8974, center=<-2.7666 0 0>, rotation=<rot 90 0 0>}]
}
constraints [
 {bone1Name=dyn_hair_bk_l_01, bone2Name=dyn_hair_bk_l_MSTR, swing1LimitDegrees=30, swing2LimitDegrees=45, twistLimitDegrees=5}
 {bone1Name=dyn_hair_bk_l_02, bone2Name=dyn_hair_bk_l_01, swing1LimitDegrees=30, swing2LimitDegrees=45, twistLimitDegrees=5}
 {bone1Name=dyn_hair_bk_l_03, bone2Name=dyn_hair_bk_l_02, swing1LimitDegrees=30, swing2LimitDegrees=45, twistLimitDegrees=5}
 {bone1Name=dyn_hair_bk_01, bone2Name=dyn_hair_bk_MSTR, swing1LimitDegrees=30, swing2LimitDegrees=45, twistLimitDegrees=5}
 {bone1Name=dyn_hair_bk_02, bone2Name=dyn_hair_bk_01, swing1LimitDegrees=30, swing2LimitDegrees=45, twistLimitDegrees=5}
 {bone1Name=dyn_hair_bk_03, bone2Name=dyn_hair_bk_02, swing1LimitDegrees=30, swing2LimitDegrees=45, twistLimitDegrees=5}
 {bone1Name=dyn_hair_bk_04, bone2Name=dyn_hair_bk_03, swing1LimitDegrees=30, swing2LimitDegrees=45, twistLimitDegrees=5}
 {bone1Name=dyn_hair_bk_r_01, bone2Name=dyn_hair_bk_r_MSTR, swing1LimitDegrees=30, swing2LimitDegrees=45, twistLimitDegrees=5}
 {bone1Name=dyn_hair_bk_r_02, bone2Name=dyn_hair_bk_r_01, swing1LimitDegrees=30, swing2LimitDegrees=45, twistLimitDegrees=5}
 {bone1Name=dyn_hair_bk_r_03, bone2Name=dyn_hair_bk_r_02, swing1LimitDegrees=30, swing2LimitDegrees=45, twistLimitDegrees=5}
 {
  bone1Name=head
  bone2Name=neck_01
  swing1LimitDegrees=45
  swing1Motion=Free
  swing2LimitDegrees=45
  swing2Motion=Free
  twistLimitDegrees=45
  twistMotion=Free
 }
 {
  bone1Name=neck_01
  bone2Name=spine_05
  swing1LimitDegrees=45
  swing1Motion=Free
  swing2LimitDegrees=45
  swing2Motion=Free
  twistLimitDegrees=45
  twistMotion=Free
 }
 {
  bone1Name=clavicle_r
  bone2Name=spine_05
  swing1LimitDegrees=45
  swing1Motion=Free
  swing2LimitDegrees=45
  swing2Motion=Free
  twistLimitDegrees=45
  twistMotion=Free
 }
 {
  bone1Name=clavicle_l
  bone2Name=spine_05
  swing1LimitDegrees=45
  swing1Motion=Free
  swing2LimitDegrees=45
  swing2Motion=Free
  twistLimitDegrees=45
  twistMotion=Free
 }
 {bone1Name=dyn_hair_bk_l_04, bone2Name=dyn_hair_bk_l_03, swing1LimitDegrees=30, swing2LimitDegrees=45, twistLimitDegrees=5}
 {
  bone1Name=dyn_hair_bk_l_04
  bone2Name=dyn_hair_bk_04
  swing1LimitDegrees=45
  swing1Motion=Locked
  swing2LimitDegrees=45
  swing2Motion=Locked
  twistLimitDegrees=45
  twistMotion=Locked
 }
 {
  bone1Name=dyn_hair_bk_r_03
  bone2Name=dyn_hair_bk_03
  swing1LimitDegrees=45
  swing1Motion=Locked
  swing2LimitDegrees=45
  swing2Motion=Locked
  twistLimitDegrees=45
  twistMotion=Locked
 }
 {
  bone1Name=dyn_hair_bk_r_MSTR
  bone2Name=head
  swing1LimitDegrees=45
  swing1Motion=Free
  swing2LimitDegrees=45
  swing2Motion=Free
  twistLimitDegrees=45
  twistMotion=Free
 }
 {
  bone1Name=dyn_hair_bk_MSTR
  bone2Name=head
  swing1LimitDegrees=45
  swing1Motion=Free
  swing2LimitDegrees=45
  swing2Motion=Free
  twistLimitDegrees=45
  twistMotion=Free
 }
 {
  bone1Name=dyn_hair_bk_l_MSTR
  bone2Name=head
  swing1LimitDegrees=45
  swing1Motion=Free
  swing2LimitDegrees=45
  swing2Motion=Free
  twistLimitDegrees=45
  twistMotion=Free
 }
 {
  bone1Name=upperarm_r
  bone2Name=clavicle_r
  swing1LimitDegrees=45
  swing1Motion=Free
  swing2LimitDegrees=45
  swing2Motion=Free
  twistLimitDegrees=45
  twistMotion=Free
 }
 {bone1Name=dyn_hair_fr_r_01, bone2Name=neck_01, swing1LimitDegrees=35, swing2LimitDegrees=10, twistLimitDegrees=5}
 {bone1Name=dyn_hair_fr_r_02, bone2Name=dyn_hair_fr_r_01, swing1LimitDegrees=30, swing2LimitDegrees=45, twistLimitDegrees=5}
 {bone1Name=dyn_hair_fr_r_03, bone2Name=dyn_hair_fr_r_02, swing1LimitDegrees=30, swing2LimitDegrees=45, twistLimitDegrees=5}
 {bone1Name=dyn_hair_fr_l_01, bone2Name=head, swing1LimitDegrees=30, swing2LimitDegrees=45, twistLimitDegrees=5}
 {bone1Name=dyn_hair_fr_l_02, bone2Name=dyn_hair_fr_l_01, swing1LimitDegrees=30, swing2LimitDegrees=45, twistLimitDegrees=5}
 {bone1Name=dyn_hair_fr_l_03, bone2Name=dyn_hair_fr_l_02, swing1LimitDegrees=30, swing2LimitDegrees=45, twistLimitDegrees=5}
 {
  bone1Name=spine_05
  bone2Name=spine_04
  swing1LimitDegrees=45
  swing1Motion=Free
  swing2LimitDegrees=45
  swing2Motion=Free
  twistLimitDegrees=45
  twistMotion=Free
 }
]
