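# Physics asset for the front-hair face accessory of the female medium Agent Rogue character, set up for dynamic hair motion.
# A four-bone left strand uses boxes and capsules, while a three-bone right strand uses capsules, each chained by tight swing and twist limits.
# The left root constrains to the head, but the right root constrains to spine_05.
# Head and spine_05 capsules provide body shapes, joined by an angularly free constraint.
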PhysicsAsset /BRCosmetics/Characters/Player/Female/Medium/Bodies/F_MED_AgentRogue/Meshes/Parts/F_MED_Agent_Rogue_FaceAcc_frontHair_Dyn
bodies {
 dyn_hair_fr_l_01 [{shapeType=Box, extentX=8, extentY=5, extentZ=2.5, center=<3.5 0 0>}]
 dyn_hair_fr_l_02 [{shapeType=Box, extentX=6, extentY=4, extentZ=2.5, center=<3 0 0>}]
 dyn_hair_fr_l_03 [{shapeType=Capsule, length=2, radius=1.5, center=<2 0 0>, rotation=<rot 90 0 0>}]
 dyn_hair_fr_l_04 [{shapeType=Capsule, length=2, radius=1, center=<1 0 0>, rotation=<rot 90 0 0>}]
 dyn_hair_fr_r_01 [{shapeType=Capsule, length=4.2, radius=2, center=<3.0873 0 0>, rotation=<rot 90 0 0>}]
 dyn_hair_fr_r_02 [{shapeType=Capsule, length=3.1884, radius=1.7291, center=<2.5783 0 0>, rotation=<rot 90 0 0>}]
 dyn_hair_fr_r_03 [{shapeType=Capsule, length=2.5773, radius=1.4227, center=<1.7331 0 0.0001>, rotation=<rot 90 0 0>}]
 head [{shapeType=Capsule, length=1.4492, radius=5.9, center=<0.4562 0.66 0>, rotation=<rot 90 0 0>}]
 spine_05 [{shapeType=Capsule, length=14.3194, radius=6.4806, center=<4.2113 -0.6359 0>}]
}
constraints [
 {bone1Name=dyn_hair_fr_l_01, bone2Name=head, swing1LimitDegrees=10, swing2LimitDegrees=25, twistLimitDegrees=5}
 {bone1Name=dyn_hair_fr_l_02, bone2Name=dyn_hair_fr_l_01, swing1LimitDegrees=10, swing2LimitDegrees=25, twistLimitDegrees=5}
 {bone1Name=dyn_hair_fr_l_03, bone2Name=dyn_hair_fr_l_02, swing1LimitDegrees=10, swing2LimitDegrees=25, twistLimitDegrees=5}
 {bone1Name=dyn_hair_fr_l_04, bone2Name=dyn_hair_fr_l_03, swing1LimitDegrees=10, swing2LimitDegrees=25, twistLimitDegrees=5}
 {bone1Name=dyn_hair_fr_r_02, bone2Name=dyn_hair_fr_r_01, swing1LimitDegrees=10, swing2LimitDegrees=15, twistLimitDegrees=5}
 {bone1Name=dyn_hair_fr_r_03, bone2Name=dyn_hair_fr_r_02, swing1LimitDegrees=10, swing2LimitDegrees=15, twistLimitDegrees=5}
 {
  bone1Name=head
  bone2Name=spine_05
  swing1LimitDegrees=45
  swing1Motion=Free
  swing2LimitDegrees=45
  swing2Motion=Free
  twistLimitDegrees=45
  twistMotion=Free
 }
 {bone1Name=dyn_hair_fr_r_01, bone2Name=spine_05, swing1LimitDegrees=10, swing2LimitDegrees=15, twistLimitDegrees=5}
]
