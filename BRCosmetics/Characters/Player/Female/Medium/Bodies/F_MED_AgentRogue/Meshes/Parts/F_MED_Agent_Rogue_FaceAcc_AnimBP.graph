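# Animation Blueprint for the female Agent Rogue face accessory part, adding secondary motion to its hair.
# It copies the attached parent's pose, layering both arms from the owner's body mesh.
# Three back-hair master bones then aim at offsets from spine_05 before RigidBody simulates in that bone's space with fixed gravity.
# The event graph lowers a gravity variable during skydiving freefall, but only an unconnected, zero-alpha front-hair RigidBody reads it.
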AnimBlueprint /BRCosmetics/Characters/Player/Female/Medium/Bodies/F_MED_AgentRogue/Meshes/Parts/F_MED_Agent_Rogue_FaceAcc_AnimBP
target_skeleton=&/F_MED_Agent_Rogue_FaceAcc_Skeleton
graphs [
 AnimGraph {nodes [
   RigidBody!(ComponentPose, Alpha=0, OverrideWorldGravity:Vector=Get gravity_override() -> (gravity_override:Vector), ExternalForce) -> (Pose) {
    bClampLinearTranslationLimitToRefPose=true
    bOverrideWorldGravity=true
    overridePhysicsAsset=&/F_MED_Agent_Rogue_FaceAcc_frontHair_Dyn
    simulationSpace=BaseBoneSpace
    alphaBoolBlend {blendOption=Linear}
    baseBoneRef {boneName=neck_01}
    componentAppliedLinearAccClamp=<1000 1000 1000>
    componentLinearAccScale=<0.3 0.3 0.2>
    componentLinearVelScale=<0.2 0.2 0.2>
    overrideWorldGravity {z=-600}
   }
   n1 = Copy Pose From Mesh(SourceMeshComponent=Get Body_Mesh())
   n2 = Copy Pose From Mesh() {bUseAttachedParent=true}
   n3 = Look At - Bone: dyn_hair_bk_l_MSTR(
    ComponentPose=Local To Component(LocalPose=Layered blend per bone(BasePose=@n2, BlendPoses_0=@n1) {layerSetup [{branchFilters [{boneName=upperarm_l}, {boneName=upperarm_r}]}]})
    Alpha=1
   ) {
    alphaBoolBlend {blendOption=Linear}
    boneToModify {boneName=dyn_hair_bk_l_MSTR}
    lookAtLocation=<-20 -15.6335 -11.854>
    lookAtTarget {boneReference {boneName=spine_05}}
    lookAt_Axis {axis=<1 0>}
   }
   n4 = Look At - Bone: dyn_hair_bk_MSTR(ComponentPose=@n3, Alpha=1) {
    alphaBoolBlend {blendOption=Linear}
    boneToModify {boneName=dyn_hair_bk_MSTR}
    lookAtLocation=<-20 -23.2983 0>
    lookAtTarget {boneReference {boneName=spine_05}}
    lookAt_Axis {axis=<1 0>}
   }
   n5 = Look At - Bone: dyn_hair_bk_r_MSTR(ComponentPose=@n4, Alpha=1) {
    alphaBoolBlend {blendOption=Linear}
    boneToModify {boneName=dyn_hair_bk_r_MSTR}
    lookAtLocation=<-20 -14.2071 22.775>
    lookAtTarget {boneReference {boneName=spine_05}}
    lookAt_Axis {axis=<1 0>}
   }
   n6 = RigidBody(ComponentPose=@n5, OverrideWorldGravity="0,0,-980") {
    bClampLinearTranslationLimitToRefPose=true
    bOverrideWorldGravity=true
    overridePhysicsAsset=&/F_MED_Agent_Rogue_FaceAcc_Physics
    simulationSpace=BaseBoneSpace
    alphaBoolBlend {blendOption=Linear}
    baseBoneRef {boneName=spine_05}
    componentAppliedLinearAccClamp=<1000 1000 1000>
    componentLinearAccScale=<0.3 0.3 0.2>
    componentLinearVelScale=<0.2 0.2 0.2>
    overrideWorldGravity {z=-600}
   }
   Output Pose(Result=Component To Local(ComponentPose=@n6)) {layerGroup=None}
  ]}
 EventGraph {nodes [
   Event BlueprintUpdateAnimation()
   > n1 = Branch(Condition=Get bIsSkyDiving())
   > n2 = Branch(Condition=Get bIsParachuteOpened())
   Set gravity_override(execute=[@n2.then, @n1.else], gravity_override="0, 0,-900")
   Set gravity_override(execute=@n2.else, gravity_override="0, 0,-300")
   Event BlueprintInitializeAnimation()
   > Set Body_Mesh(Body_Mesh=GetOwnerSkeletalMeshComponent())
   TryGetPawnOwner()
  ]}
]
variables [
 Body_Mesh {type=SkeletalMeshComponent}
 gravity_override {type=Vector, default=<0 0 -900>}
]
